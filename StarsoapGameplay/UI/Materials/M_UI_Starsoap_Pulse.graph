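# M_UI_Starsoap_Pulse is a translucent UI material that draws repeating animated pulse bands for the Starsoap interface.
# The red channel of MF_PolarCoordinates, likely radial distance, scrolls with time by PulseSpeed and repeats by PulseRepetition through Frac.
# Each band combines a half-strength gradient fill with a thin border.
# A soft window on the same coordinate fades both ends, and PulseActive scales the result.
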Material /StarsoapGameplay/UI/Materials/M_UI_Starsoap_Pulse
graph {
 properties {
  bUsedWithStaticMesh=false
  blendMode=BLEND_Translucent
  materialDomain=MD_UI
  numCustomizedUVs=3
  pixelDepthOffsetMode=PDOM_Legacy
  usedShadingModels=Default Lit
 }
 nodes [
  Constant() {r=1}
  n1 = ComponentMask(None=MaterialFunctionCall!(UV Coord, LenghtScale, RadialScale) {materialFunction=/Game/Global/MaterialFunctions/TechArt/MF_PolarCoordinates}.Result) {r=true}
  n2 = Multiply(A=@n1) {constB=2}
  n3 = Min(A=Multiply(A=Add(A=@n2) {constB=-0.5}) {constB=6}, B=Multiply(A=OneMinus(None=@n2)) {constB=3})
  n4 = Multiply(A=Time(), B=ScalarParameter() {defaultValue=-0.1, group=Pulse, parameterName=PulseSpeed})
  n5 = Frac(None=Multiply(A=Add(A=@n1, B=@n4), B=ScalarParameter() {defaultValue=12, group=Pulse, parameterName=PulseRepetition}))
  n6 = Multiply(A=OneMinus(None=@n5)) {constB=3, desc=border thickness}
  n7 = Multiply(A=Min(A=@n5, B=@n6)) {constB=0.5, desc=fillopacity}
  n8 = Saturate(None=Min(A=@n6, B=OneMinus(None=@n6)))
  Multiply(A=ScalarParameter() {defaultValue=1, group=Pulse, parameterName=PulseActive}, B=Multiply(A=Saturate(None=@n3), B=Add(A=@n7, B=@n8)))
 ]
}
