WidgetBlueprint /StarsoapGameplay/UI/Proto/WBP_StarsoapActivateInstigator
animations [
 Intro {
  binding_count=8
  length_seconds=0.95
  object=&self:Intro
  sequence {bindings [
    SizeBox_Main {
     id=#1
     possessed_class=SizeBox
     tracks [{
       type=MovieSceneFloatTrack
       display_name=Max Desired Width
       sections [{
         type=MovieSceneFloatSection
         end_frame=13
         start_frame=0
         channels [None {type=MovieSceneScriptingFloatChannel, keys [[0,128],[7,128],[13,800]]}]
        }]
      }]
    }
    Image_DecoTop {
     id=#2
     possessed_class=Image
     tracks [{
       type=MovieSceneFloatTrack
       display_name=Render Opacity
       sections [{type=MovieSceneFloatSection, end_frame=11, start_frame=0, channels [None {type=MovieSceneScriptingFloatChannel, keys [[0,0],[7,0],[11,1]]}]}]
      }]
    }
    Text_IntroText_1 {
     id=#3
     possessed_class=CommonTextBlock
     tracks [{
       type=MovieScene2DTransformTrack
       sections [{
         type=MovieScene2DTransformSection
         end_frame=12
         start_frame=0
         channels [
          Translation.X {type=MovieSceneScriptingFloatChannel, default=0}
          Translation.Y {type=MovieSceneScriptingFloatChannel, keys [[9,25],[12,0]]}
          Scale.X {type=MovieSceneScriptingFloatChannel, keys [[0,0.5],[6,0.5],[9,1.5],[12,1]]}
          Scale.Y {type=MovieSceneScriptingFloatChannel, keys [[0,0.5],[6,0.5],[9,1.5],[12,1]]}
         ]
        }]
      }]
    }
    Overlay_Secondary {
     id=#4
     possessed_class=Overlay
     tracks [
      {
       type=MovieSceneFloatTrack
       display_name=Render Opacity
       sections [{
         type=MovieSceneFloatSection
         end_frame=19
         start_frame=0
         channels [None {type=MovieSceneScriptingFloatChannel, keys [[0,0],[7,0],[18,1],[19,1]]}]
        }]
      }
      {
       type=MovieScene2DTransformTrack
       sections [{
         type=MovieScene2DTransformSection
         end_frame=19
         start_frame=0
         channels [
          Translation.X {type=MovieSceneScriptingFloatChannel, const [0,1]}
          Translation.Y {type=MovieSceneScriptingFloatChannel, const [0,1]}
          Angle {type=MovieSceneScriptingFloatChannel}
          Scale.X {type=MovieSceneScriptingFloatChannel, keys [[0,0.2],[7,0.2],[18,1],[19,1]]}
          Scale.Y {type=MovieSceneScriptingFloatChannel, keys [[0,0.2],[7,0.2],[18,1],[19,1]]}
          Shear.X {type=MovieSceneScriptingFloatChannel}
          Shear.Y {type=MovieSceneScriptingFloatChannel}
         ]
        }]
      }
     ]
    }
    Overlay_Main {
     id=#5
     possessed_class=Overlay
     tracks [
      {
       type=MovieSceneFloatTrack
       display_name=Render Opacity
       sections [{type=MovieSceneFloatSection, end_frame=8, start_frame=0, channels [None {type=MovieSceneScriptingFloatChannel, keys [[0,0],[4,0],[8,1]]}]}]
      }
      {
       type=MovieScene2DTransformTrack
       sections [{
         type=MovieScene2DTransformSection
         end_frame=11
         start_frame=4
         channels [
          Translation.X {type=MovieSceneScriptingFloatChannel, const [0,2]}
          Translation.Y {type=MovieSceneScriptingFloatChannel, keys [[4,-100],[7,30],[9,0]]}
          Scale.X {type=MovieSceneScriptingFloatChannel, keys [[4,0.5],[7,1]]}
          Scale.Y {type=MovieSceneScriptingFloatChannel, keys [[4,0.5],[7,1]]}
         ]
        }]
      }
     ]
    }
    Text_IntroText_2 {
     id=#6
     possessed_class=CommonTextBlock
     tracks [
      {
       type=MovieSceneFloatTrack
       display_name=Render Opacity
       sections [{type=MovieSceneFloatSection, end_frame=12, start_frame=0, channels [None {type=MovieSceneScriptingFloatChannel, keys [[0,0],[8,0],[12,1]]}]}]
      }
      {
       type=MovieScene2DTransformTrack
       sections [{
         type=MovieScene2DTransformSection
         end_frame=15
         start_frame=0
         channels [
          Scale.X {type=MovieSceneScriptingFloatChannel, keys [[0,0],[8,0],[12,1.2],[15,1]]}
          Scale.Y {type=MovieSceneScriptingFloatChannel, keys [[0,0],[8,0],[12,1.2],[15,1]]}
         ]
        }]
      }
     ]
    }
    GridPanel_0 {
     id=#7
     possessed_class=GridPanel
     tracks [
      {
       type=MovieSceneFloatTrack
       display_name=Render Opacity
       sections [{type=MovieSceneFloatSection, end_frame=4, start_frame=0, channels [None {type=MovieSceneScriptingFloatChannel, keys [[0,0],[3,1]]}]}]
      }
      {
       type=MovieScene2DTransformTrack
       sections [{
         type=MovieScene2DTransformSection
         end_frame=0
         start_frame=0
         channels [
          Translation.X {type=MovieSceneScriptingFloatChannel, const [0,1]}
          Translation.Y {type=MovieSceneScriptingFloatChannel, const [0,1]}
          Scale.X {type=MovieSceneScriptingFloatChannel, const [1,1]}
          Scale.Y {type=MovieSceneScriptingFloatChannel, const [1,1]}
         ]
        }]
      }
     ]
    }
    Overlay_Icon {
     id=#8
     possessed_class=Overlay
     tracks [{
       type=MovieScene2DTransformTrack
       sections [
        {
         type=MovieScene2DTransformSection
         end_frame=0
         start_frame=0
         channels [
          Translation.X {type=MovieSceneScriptingFloatChannel, const [0,1]}
          Translation.Y {type=MovieSceneScriptingFloatChannel, const [0,1]}
          Scale.X {type=MovieSceneScriptingFloatChannel, const [1,1]}
          Scale.Y {type=MovieSceneScriptingFloatChannel, const [1,1]}
         ]
        }
        {
         type=MovieScene2DTransformSection
         end_frame=15
         start_frame=0
         channels [
          Translation.X {type=MovieSceneScriptingFloatChannel}
          Translation.Y {type=MovieSceneScriptingFloatChannel}
          Angle {type=MovieSceneScriptingFloatChannel, default=0}
          Scale.X {type=MovieSceneScriptingFloatChannel, keys [[0,0.2],[7,1.3],[11,1]]}
          Scale.Y {type=MovieSceneScriptingFloatChannel, keys [[0,0.2],[7,1.3],[11,1]]}
         ]
        }
       ]
      }]
    }
   ]}
 }
 Minimize {
  binding_count=7
  length_seconds=0.55
  object=&self:Minimize
  sequence {bindings [
    GridSlot_5 {
     id=#9
     parent_id=#10
     possessed_class=GridSlot
     tracks [{
       type=MovieSceneDoubleVectorTrack
       display_name=Nudge
       sections [{
         type=MovieSceneDoubleVectorSection
         end_frame=11
         start_frame=-1
         channels [
          Vector.X {type=MovieSceneScriptingDoubleChannel, const [0,1]}
          Vector.Y {type=MovieSceneScriptingDoubleChannel, keys [[-0.0003,-60],[4,-60],[10,-130],[11,-130]]}
         ]
        }]
      }]
    }
    Overlay_Main {
     id=#11
     possessed_class=Overlay
     tracks [
      {
       type=MovieSceneFloatTrack
       display_name=Render Opacity
       sections [{type=MovieSceneFloatSection, end_frame=10, start_frame=0, channels [None {type=MovieSceneScriptingFloatChannel, keys [[0,1],[6,1],[10,0]]}]}]
      }
      {
       type=MovieScene2DTransformTrack
       sections [{
         type=MovieScene2DTransformSection
         end_frame=10
         start_frame=0
         channels [
          Translation.X {type=MovieSceneScriptingFloatChannel, const [0,1]}
          Translation.Y {type=MovieSceneScriptingFloatChannel, keys [[0,0],[6,0],[10,-75]]}
          Scale.X {type=MovieSceneScriptingFloatChannel}
          Scale.Y {type=MovieSceneScriptingFloatChannel}
         ]
        }]
      }
     ]
    }
    Overlay_Secondary {id=#10, possessed_class=Overlay, tracks []}
    Image_Pulse {
     id=#12
     possessed_class=Image
     tracks [
      {
       type=MovieSceneFloatTrack
       display_name=Render Opacity
       sections [{type=MovieSceneFloatSection, end_frame=10, start_frame=0, channels [None {type=MovieSceneScriptingFloatChannel, keys [[0,0],[7,0],[10,1]]}]}]
      }
      {
       type=MovieSceneByteTrack
       display_name=Visibility
       sections [{type=MovieSceneByteSection, end_frame=7, start_frame=0, channels [None {type=MovieSceneScriptingByteChannel, keys [[0,1],[7,4]]}]}]
      }
     ]
    }
    SizeBox_Main {
     id=#13
     possessed_class=SizeBox
     tracks [
      {
       type=MovieSceneFloatTrack
       display_name=Max Desired Height
       sections [{type=MovieSceneFloatSection, end_frame=10, start_frame=-1, channels [None {type=MovieSceneScriptingFloatChannel, keys [[0,112],[10,0]]}]}]
      }
      {
       type=MovieSceneFloatTrack
       display_name=Max Desired Width
       sections [{type=MovieSceneFloatSection, end_frame=10, start_frame=-1, channels [None {type=MovieSceneScriptingFloatChannel, keys [[0,800],[10,0]]}]}]
      }
      {
       type=MovieSceneFloatTrack
       display_name=Render Opacity
       sections [{type=MovieSceneFloatSection, end_frame=6, start_frame=-1, channels [None {type=MovieSceneScriptingFloatChannel, keys [[0,1],[3,1],[6,0]]}]}]
      }
      {
       type=MovieScene2DTransformTrack
       sections [{
         type=MovieScene2DTransformSection
         end_frame=6
         start_frame=0
         channels [
          Translation.X {type=MovieSceneScriptingFloatChannel, default=0}
          Translation.Y {type=MovieSceneScriptingFloatChannel, default=0}
          Scale.X {type=MovieSceneScriptingFloatChannel, keys [[0,1],[3,1.2],[6,0.2]]}
          Scale.Y {type=MovieSceneScriptingFloatChannel, keys [[0,1],[3,1.2],[6,0.2]]}
         ]
        }]
      }
      {
       type=MovieSceneByteTrack
       display_name=Visibility
       sections [{type=MovieSceneByteSection, end_frame=10, start_frame=0, channels [None {type=MovieSceneScriptingByteChannel, keys [[0,3],[10,1]]}]}]
      }
     ]
    }
    Overlay_Icon {
     id=#14
     possessed_class=Overlay
     tracks [{
       type=MovieScene2DTransformTrack
       sections [{
         type=MovieScene2DTransformSection
         end_frame=10
         start_frame=-1
         channels [
          Translation.X {type=MovieSceneScriptingFloatChannel}
          Translation.Y {type=MovieSceneScriptingFloatChannel}
          Scale.X {type=MovieSceneScriptingFloatChannel, keys [[-0.0003,1],[4,1],[7,1.3],[10,1]]}
          Scale.Y {type=MovieSceneScriptingFloatChannel, keys [[-0.0003,1],[4,1],[7,1.3],[10,1]]}
         ]
        }]
      }]
    }
    Image_DecoTop {
     id=#15
     possessed_class=Image
     tracks [{
       type=MovieSceneFloatTrack
       display_name=Render Opacity
       sections [{type=MovieSceneFloatSection, end_frame=10, start_frame=0, channels [None {type=MovieSceneScriptingFloatChannel, keys [[0,1],[4,1],[7,0]]}]}]
      }]
    }
   ]}
 }
 Outro {
  binding_count=2
  length_seconds=0.3
  object=&self:Outro
  sequence {bindings [
    GridPanel_0 {
     id=#16
     possessed_class=GridPanel
     tracks [
      {
       type=MovieSceneFloatTrack
       display_name=Render Opacity
       sections [{type=MovieSceneFloatSection, end_frame=6, start_frame=0, channels [None {type=MovieSceneScriptingFloatChannel, keys [[0,1],[5,0],[6,0]]}]}]
      }
      {
       type=MovieScene2DTransformTrack
       sections [{
         type=MovieScene2DTransformSection
         end_frame=5
         start_frame=0
         channels [
          Translation.X {type=MovieSceneScriptingFloatChannel, const [0,1]}
          Translation.Y {type=MovieSceneScriptingFloatChannel, keys [[0,0],[5,-75]]}
          Scale.X {type=MovieSceneScriptingFloatChannel, keys [[0,1],[5,0.5]]}
          Scale.Y {type=MovieSceneScriptingFloatChannel, keys [[0,1],[5,0.5]]}
         ]
        }]
      }
     ]
    }
    Overlay_Secondary {
     id=#17
     possessed_class=Overlay
     tracks [{
       type=MovieScene2DTransformTrack
       sections [{
         type=MovieScene2DTransformSection
         end_frame=5
         start_frame=0
         channels [
          Translation.X {type=MovieSceneScriptingFloatChannel, const [0,1]}
          Translation.Y {type=MovieSceneScriptingFloatChannel, keys [[0,0],[5,-100]]}
          Angle {type=MovieSceneScriptingFloatChannel}
          Scale.X {type=MovieSceneScriptingFloatChannel, keys [[0,1],[5,0.2]]}
          Scale.Y {type=MovieSceneScriptingFloatChannel, keys [[0,1],[5,0.2]]}
          Shear.X {type=MovieSceneScriptingFloatChannel}
          Shear.Y {type=MovieSceneScriptingFloatChannel}
         ]
        }]
      }]
    }
   ]}
 }
]
graphs [
 EventGraph {nodes [
   Event Construct()
   > PlayAnimation(InAnimation=Get Intro())
   > Delay(Duration=5)
   > PlayAnimation(InAnimation=Get Minimize() -> (Minimize:Widget Animation Object Reference))
   > Delay(Duration=5)
   > PlayAnimation(InAnimation=Get Outro() -> (Outro:Widget Animation Object Reference))
   "Animation Finished (Outro)"!() -> (OutputDelegate:Delegate, then:Exec) {is=title K2Node_WidgetAnimationEvent}
   > ClearEventState!(
    execute
    self:Gameplay Event Router Component Object Reference=GetGlobalEventRouter()
    EventType:Script Struct Object Reference=/Script/BattleRoyaleGameplayAlerts.FortEvent_CurrentGameplayAlertChanged
   ) -> (then)
   Event Tick()
   Event OnInitialized()
  ]}
 SetPlayerPortrait {nodes [
   n1 = SetPlayerPortrait!() -> (then, NewParam:Texture Object Reference) {is=title K2Node_FunctionEntry}
   > n2 = GetDynamicMaterial(self=Get Image_HeaderIcon() -> (Image_HeaderIcon:Image Object Reference))
   > SetTextureParameterValue(self=@n2, ParameterName=SmallPreview, Value=@n1)
  ]}
]
variables []
widgets {
 info {inheritedWidgetCount=0, parentClass=&UserWidget, rootWidgetClass=&GridPanel, widgetCount=21}
 widgets [
  GridPanel_0 {widgetClassPath=&GridPanel, widget=GridPanel_0}
  Overlay_Icon {
   widgetClassPath=&Overlay
   parent=GridPanel_0
   widget=Overlay_Icon
   slot {class=GridSlot, properties {column=1, horizontalAlignment=HAlign_Center, layer=1, verticalAlignment=VAlign_Top, nudge=<0 -6>}}
  }
  Image_Pulse {
   widgetClassPath=&Image
   parent=Overlay_Icon
   widget=Image_Pulse
   properties {
    visibility=Collapsed
    brush {
     bIsDynamicallyLoaded=false
     drawAs=Image
     imageType=FullColor
     mirroring=NoMirror
     resourceName=None
     resourceObject=/StarsoapGameplay/UI/Materials/M_UI_Starsoap_Pulse
     tiling=NoTile
     imageSize=<160 160>
     margin=<ltrb 0 0 0 0>
     outlineSettings {
      bUseBrushTransparency=false
      roundingType=HalfHeightRadius
      width=0
      color {colorUseRule=UseColor_Specified, specifiedColor=<rgba 0 0 0 0>}
      cornerRadii=<0 0 0 0>
     }
     tintColor {colorUseRule=UseColor_Specified, specifiedColor=<rgba 1 1 1 1>}
     uVRegion {bIsValid=false, max=<0 0>, min=<0 0>}
    }
    colorAndOpacity=<rgba 1 0.1981 0.0048 1>
   }
   slot {class=OverlaySlot, properties {horizontalAlignment=HAlign_Center, verticalAlignment=VAlign_Center}}
  }
  Image_HeaderIcon {
   widgetClassPath=&Image
   bIsVariable=true
   parent=Overlay_Icon
   widget=Image_HeaderIcon
   properties {
    visibility=SelfHitTestInvisible
    brush {
     bIsDynamicallyLoaded=false
     drawAs=Image
     imageType=FullColor
     mirroring=NoMirror
     resourceName=None
     resourceObject=/StarsoapGameplay/UI/MaterialInstances/MI_UI_Starsoap_Portrait
     tiling=NoTile
     imageSize=<160 160>
     margin=<ltrb 0 0 0 0>
     outlineSettings {
      bUseBrushTransparency=false
      roundingType=HalfHeightRadius
      width=0
      color {colorUseRule=UseColor_Specified, specifiedColor=<rgba 0 0 0 0>}
      cornerRadii=<0 0 0 0>
     }
     tintColor {colorUseRule=UseColor_Specified, specifiedColor=<rgba 1 1 1 1>}
     uVRegion {bIsValid=false, max=<0 0>, min=<0 0>}
    }
   }
   slot {class=OverlaySlot, properties {horizontalAlignment=HAlign_Center, verticalAlignment=VAlign_Center}}
  }
  Overlay_Main {
   widgetClassPath=&Overlay
   parent=GridPanel_0
   widget=Overlay_Main
   slot {class=GridSlot, properties {column=1, horizontalAlignment=HAlign_Center, row=1, verticalAlignment=VAlign_Center, nudge=<0 -40.7672>}}
  }
  Image_Background {
   widgetClassPath=&Image
   parent=Overlay_Main
   widget=Image_Background
   properties {
    brush {
     bIsDynamicallyLoaded=false
     drawAs=Image
     imageType=FullColor
     mirroring=NoMirror
     resourceName=None
     resourceObject=/StarsoapGameplay/UI/MaterialInstances/MI_UI_Starsoap_BackNegative
     tiling=NoTile
     imageSize=<1 1>
     margin=<ltrb 0 0 0 0>
     outlineSettings {
      bUseBrushTransparency=false
      roundingType=HalfHeightRadius
      width=0
      color {colorUseRule=UseColor_Specified, specifiedColor=<rgba 0 0 0 0>}
      cornerRadii=<0 0 0 0>
     }
     tintColor {colorUseRule=UseColor_Specified, specifiedColor=<rgba 1 1 1 1>}
     uVRegion {bIsValid=false, max=<0 0>, min=<0 0>}
    }
    colorAndOpacity=<rgba 1 1 1 0.95>
   }
   slot {class=OverlaySlot, properties {horizontalAlignment=HAlign_Fill, verticalAlignment=VAlign_Fill, padding=<ltrb 5 10 5 22>}}
  }
  Image_FrameTop {
   widgetClassPath=&Image
   parent=Overlay_Main
   widget=Image_FrameTop
   properties {brush {
     bIsDynamicallyLoaded=false
     drawAs=Image
     imageType=FullColor
     mirroring=NoMirror
     resourceName=None
     resourceObject=/StarsoapGameplay/UI/Proto/Textures/T_UI_Starsoap_FrameTop
     tiling=NoTile
     imageSize=<1 16>
     margin=<ltrb 0 0 0 0>
     outlineSettings {
      bUseBrushTransparency=false
      roundingType=HalfHeightRadius
      width=0
      color {colorUseRule=UseColor_Specified, specifiedColor=<rgba 0 0 0 0>}
      cornerRadii=<0 0 0 0>
     }
     tintColor {colorUseRule=UseColor_Specified, specifiedColor=<rgba 1 1 1 1>}
     uVRegion {bIsValid=false, max=<0 0>, min=<0 0>}
    }}
   slot {class=OverlaySlot, properties {horizontalAlignment=HAlign_Fill, padding=<ltrb 40 0 40 0>}}
  }
  Image_FrameBottom {
   widgetClassPath=&Image
   parent=Overlay_Main
   widget=Image_FrameBottom
   properties {brush {
     bIsDynamicallyLoaded=false
     drawAs=Image
     imageType=FullColor
     mirroring=NoMirror
     resourceName=None
     resourceObject=/StarsoapGameplay/UI/Proto/Textures/T_UI_Starsoap_FrameBottom
     tiling=NoTile
     imageSize=<1 32>
     margin=<ltrb 0 0 0 0>
     outlineSettings {
      bUseBrushTransparency=false
      roundingType=HalfHeightRadius
      width=0
      color {colorUseRule=UseColor_Specified, specifiedColor=<rgba 0 0 0 0>}
      cornerRadii=<0 0 0 0>
     }
     tintColor {colorUseRule=UseColor_Specified, specifiedColor=<rgba 1 1 1 1>}
     uVRegion {bIsValid=false, max=<0 0>, min=<0 0>}
    }}
   slot {class=OverlaySlot, properties {horizontalAlignment=HAlign_Fill, verticalAlignment=VAlign_Bottom}}
  }
  Image_DecoBottom {
   widgetClassPath=&Image
   parent=Overlay_Main
   widget=Image_DecoBottom
   properties {brush {
     bIsDynamicallyLoaded=false
     drawAs=Image
     imageType=FullColor
     mirroring=NoMirror
     resourceName=None
     resourceObject=/StarsoapGameplay/UI/Proto/Textures/T_UI_Starsoap_DecoBottom
     tiling=NoTile
     imageSize=<64 32>
     margin=<ltrb 0 0 0 0>
     outlineSettings {
      bUseBrushTransparency=false
      roundingType=HalfHeightRadius
      width=0
      color {colorUseRule=UseColor_Specified, specifiedColor=<rgba 0 0 0 0>}
      cornerRadii=<0 0 0 0>
     }
     tintColor {colorUseRule=UseColor_Specified, specifiedColor=<rgba 1 1 1 1>}
     uVRegion {bIsValid=false, max=<0 0>, min=<0 0>}
    }}
   slot {class=OverlaySlot, properties {horizontalAlignment=HAlign_Center, verticalAlignment=VAlign_Bottom}}
  }
  ScaleBox_RightDeco {
   widgetClassPath=&ScaleBox
   parent=Overlay_Main
   widget=ScaleBox_RightDeco
   properties {stretch=ScaleToFitY}
   slot {class=OverlaySlot, properties {horizontalAlignment=HAlign_Right, verticalAlignment=VAlign_Fill, padding=<ltrb 0 0 -8 16>}}
  }
  Image_DecoRight {
   widgetClassPath=&Image
   parent=ScaleBox_RightDeco
   widget=Image_DecoRight
   properties {brush {
     bIsDynamicallyLoaded=false
     drawAs=Image
     imageType=FullColor
     mirroring=NoMirror
     resourceName=None
     resourceObject=/StarsoapGameplay/UI/Proto/Textures/T_UI_Starsoap_DecoRight
     tiling=NoTile
     imageSize=<40 80>
     margin=<ltrb 0 0 0 0>
     outlineSettings {
      bUseBrushTransparency=false
      roundingType=HalfHeightRadius
      width=0
      color {colorUseRule=UseColor_Specified, specifiedColor=<rgba 0 0 0 0>}
      cornerRadii=<0 0 0 0>
     }
     tintColor {colorUseRule=UseColor_Specified, specifiedColor=<rgba 1 1 1 1>}
     uVRegion {bIsValid=false, max=<0 0>, min=<0 0>}
    }}
   slot {class=ScaleBoxSlot, properties {horizontalAlignment=HAlign_Fill, verticalAlignment=VAlign_Fill}}
  }
  ScaleBox_LeftDeco {
   widgetClassPath=&ScaleBox
   parent=Overlay_Main
   widget=ScaleBox_LeftDeco
   properties {stretch=ScaleToFitY}
   slot {class=OverlaySlot, properties {verticalAlignment=VAlign_Fill, padding=<ltrb -8 0 0 16>}}
  }
  Image_DecoLeft {
   widgetClassPath=&Image
   parent=ScaleBox_LeftDeco
   widget=Image_DecoLeft
   properties {brush {
     bIsDynamicallyLoaded=false
     drawAs=Image
     imageType=FullColor
     mirroring=NoMirror
     resourceName=None
     resourceObject=/StarsoapGameplay/UI/Proto/Textures/T_UI_Starsoap_DecoLeft
     tiling=NoTile
     imageSize=<40 80>
     margin=<ltrb 0 0 0 0>
     outlineSettings {
      bUseBrushTransparency=false
      roundingType=HalfHeightRadius
      width=0
      color {colorUseRule=UseColor_Specified, specifiedColor=<rgba 0 0 0 0>}
      cornerRadii=<0 0 0 0>
     }
     tintColor {colorUseRule=UseColor_Specified, specifiedColor=<rgba 1 1 1 1>}
     uVRegion {bIsValid=false, max=<0 0>, min=<0 0>}
    }}
   slot {class=ScaleBoxSlot, properties {horizontalAlignment=HAlign_Fill, verticalAlignment=VAlign_Fill}}
  }
  Image_DecoTop {
   widgetClassPath=&Image
   parent=Overlay_Main
   widget=Image_DecoTop
   properties {brush {
     bIsDynamicallyLoaded=false
     drawAs=Image
     imageType=FullColor
     mirroring=NoMirror
     resourceName=None
     resourceObject=/StarsoapGameplay/UI/Proto/Textures/T_UI_Starsoap_DecoTop
     tiling=NoTile
     imageSize=<256 64>
     margin=<ltrb 0 0 0 0>
     outlineSettings {
      bUseBrushTransparency=false
      roundingType=HalfHeightRadius
      width=0
      color {colorUseRule=UseColor_Specified, specifiedColor=<rgba 0 0 0 0>}
      cornerRadii=<0 0 0 0>
     }
     tintColor {colorUseRule=UseColor_Specified, specifiedColor=<rgba 1 1 1 1>}
     uVRegion {bIsValid=false, max=<0 0>, min=<0 0>}
    }}
   slot {class=OverlaySlot, properties {horizontalAlignment=HAlign_Center, padding=<ltrb 0 -42 0 0>}}
  }
  SizeBox_Main {
   widgetClassPath=&SizeBox
   parent=Overlay_Main
   widget=SizeBox_Main
   properties {bOverride_MaxDesiredWidth=true, maxDesiredHeight=112, maxDesiredWidth=800}
   slot {class=OverlaySlot, properties {horizontalAlignment=HAlign_Center, verticalAlignment=VAlign_Center, padding=<ltrb 80 24 80 28>}}
  }
  VerticalBox_0 {
   widgetClassPath=&VerticalBox
   parent=SizeBox_Main
   widget=VerticalBox_0
   properties {bOverride_Cursor=true}
   slot {class=SizeBoxSlot}
  }
  Text_IntroText_1 {
   widgetClassPath=&CommonTextBlock
   bIsVariable=true
   parent=VerticalBox_0
   widget=Text_IntroText_1
   properties {
    justification=Center
    text=You Are
    font {
     bForceMonospaced=false
     bMaterialIsStencil=false
     fontMaterial=null
     fontObject=&BurbankSmall
     letterSpacing=0
     monospacedWidth=1
     size=21
     skewAmount=0
     typefaceFontName=Black
     outlineSettings {
      bApplyOutlineToDropShadows=false
      bMiteredCorners=false
      bSeparateFillAlpha=false
      outlineMaterial=null
      outlineSize=2
      outlineColor=<rgba 0 0 0 1>
     }
    }
    shadowColorAndOpacity=<rgba 0 0 0 1>
    shadowOffset=<0 2>
   }
   slot {class=VerticalBoxSlot, properties {horizontalAlignment=HAlign_Center}}
  }
  Text_IntroText_2 {
   widgetClassPath=&CommonTextBlock
   bIsVariable=true
   parent=VerticalBox_0
   widget=Text_IntroText_2
   properties {
    justification=Center
    text="Excommunicado!"
    textTransformPolicy=ToUpper
    colorAndOpacity {colorUseRule=UseColor_Specified, specifiedColor=<rgba 0.8632 0.5972 0.007 1>}
    font {
     bForceMonospaced=false
     bMaterialIsStencil=false
     fontMaterial=null
     fontObject=&Burbank
     letterSpacing=0
     monospacedWidth=1
     size=42
     skewAmount=0
     typefaceFontName=BurbankRegular-Black
     outlineSettings {
      bApplyOutlineToDropShadows=false
      bMiteredCorners=false
      bSeparateFillAlpha=false
      outlineMaterial=null
      outlineSize=2
      outlineColor=<rgba 0 0 0 1>
     }
    }
    shadowColorAndOpacity=<rgba 0 0 0 1>
    shadowOffset=<0 2>
   }
   slot {class=VerticalBoxSlot, properties {horizontalAlignment=HAlign_Center, padding=<ltrb 0 -16 0 0>}}
  }
  Overlay_Secondary {
   widgetClassPath=&Overlay
   parent=GridPanel_0
   widget=Overlay_Secondary
   properties {visibility=HitTestInvisible}
   slot {
    class=GridSlot
    properties {column=1, horizontalAlignment=HAlign_Center, layer=2, row=2, verticalAlignment=VAlign_Center, nudge=<0 -60>, padding=<ltrb 0 16 0 0>}
   }
  }
  Image_SecondaryBack {
   widgetClassPath=&Image
   parent=Overlay_Secondary
   widget=Image_SecondaryBack
   properties {
    brush {
     bIsDynamicallyLoaded=false
     drawAs=Box
     imageType=FullColor
     mirroring=NoMirror
     resourceName=None
     resourceObject=/StarsoapGameplay/UI/Proto/Textures/T_UI_Starsoap_SubGradient
     tiling=NoTile
     imageSize=<200 48>
     margin=<ltrb 0.1 0 0.1 0>
     outlineSettings {
      bUseBrushTransparency=false
      roundingType=HalfHeightRadius
      width=0
      color {colorUseRule=UseColor_Specified, specifiedColor=<rgba 0 0 0 0>}
      cornerRadii=<0 0 0 0>
     }
     tintColor {colorUseRule=UseColor_Specified, specifiedColor=<rgba 1 1 1 1>}
     uVRegion {bIsValid=false, max=<0 0>, min=<0 0>}
    }
    colorAndOpacity=<rgba 1 1 1 0.9>
   }
   slot {class=OverlaySlot, properties {horizontalAlignment=HAlign_Fill, verticalAlignment=VAlign_Fill}}
  }
  Text_SecondaryText {
   widgetClassPath=&CommonTextBlock
   bIsVariable=true
   parent=Overlay_Secondary
   widget=Text_SecondaryText
   properties {
    justification=Center
    lineHeightPercentage=0.8
    text="Location Revealed\r\nEscape with faster reload speed and infinite sprint."
    colorAndOpacity {colorUseRule=UseColor_Specified, specifiedColor=<rgba 1 0.8308 0.1912 1>}
    font {
     bForceMonospaced=false
     bMaterialIsStencil=false
     fontMaterial=null
     fontObject=&BurbankSmall
     letterSpacing=0
     monospacedWidth=1
     size=19.5
     skewAmount=0
     typefaceFontName=Black
     outlineSettings {
      bApplyOutlineToDropShadows=false
      bMiteredCorners=false
      bSeparateFillAlpha=false
      outlineMaterial=null
      outlineSize=2
      outlineColor=<rgba 0 0 0 1>
     }
    }
    shadowColorAndOpacity=<rgba 0 0 0 1>
    shadowOffset=<0 2>
   }
   slot {class=OverlaySlot, properties {horizontalAlignment=HAlign_Center, verticalAlignment=VAlign_Center, padding=<ltrb 40 8 40 10>}}
  }
 ]
}
