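# Prototype Starsoap alert widget announcing in red text that another player is "Excommunicated" and can be tracked down for a reward.
# A grid stacks an icon with a dots overlay above headline and subtext, the player name being an unbound placeholder.
# Construct plays Intro, animating image material parameters and bouncing the icon and staggered text in.
# OnInitialized waits five seconds, then clears the CurrentGameplayAlertChanged state on the global event router.
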WidgetBlueprint /StarsoapGameplay/UI/Proto/WBP_StarsoapActivateOther
animations [Intro {
  binding_count=4
  length_seconds=0.75
  object=&self:Intro
  sequence {bindings [
    Image_HomerIcon {
     id=#1
     possessed_class=Image
     tracks [
      {
       type=MovieSceneWidgetMaterialTrack
       sections [{
         type=MovieSceneParameterSection
         end_frame=15
         start_frame=0
         channels [
          IconScale_Character {type=MovieSceneScriptingFloatChannel, keys [[0,0],[7,0.9]]}
          Scale Factor {type=MovieSceneScriptingFloatChannel, keys [[0,-1],[7,0],[10,0]]}
          Intensity {type=MovieSceneScriptingFloatChannel, keys [[0,0],[7,0.3],[10,0.12]]}
          Intensity 2 {type=MovieSceneScriptingFloatChannel, keys [[0,0],[7,0.1],[10,0.2]]}
         ]
        }]
      }
      {
       type=MovieScene2DTransformTrack
       sections [{
         type=MovieScene2DTransformSection
         end_frame=15
         start_frame=0
         channels [
          Translation.X {type=MovieSceneScriptingFloatChannel, const [0,3]}
          Translation.Y {type=MovieSceneScriptingFloatChannel, keys [[0,128],[8,-20],[13,3]]}
         ]
        }]
      }
     ]
    }
    Dots {
     id=#2
     possessed_class=Image
     tracks [{
       type=MovieSceneWidgetMaterialTrack
       sections [{type=MovieSceneParameterSection, end_frame=15, start_frame=0, channels [Opacity {type=MovieSceneScriptingFloatChannel, keys [[5,0],[8,1]]}]}]
      }]
    }
    Text_MainText {
     id=#3
     possessed_class=CommonTextBlock
     tracks [{
       type=MovieScene2DTransformTrack
       sections [{
         type=MovieScene2DTransformSection
         end_frame=15
         start_frame=0
         channels [
          Translation.X {type=MovieSceneScriptingFloatChannel, default=0}
          Translation.Y {type=MovieSceneScriptingFloatChannel, keys [[6,40],[10,-40],[13,0]]}
          Scale.X {type=MovieSceneScriptingFloatChannel, keys [[6,0],[10,1],[13,1]]}
          Scale.Y {type=MovieSceneScriptingFloatChannel, keys [[6,0],[10,1.5],[13,1]]}
         ]
        }]
      }]
    }
    Text_MainText_1 {
     id=#4
     possessed_class=CommonTextBlock
     tracks [{
       type=MovieScene2DTransformTrack
       sections [{
         type=MovieScene2DTransformSection
         end_frame=15
         start_frame=0
         channels [
          Translation.X {type=MovieSceneScriptingFloatChannel, default=0}
          Translation.Y {type=MovieSceneScriptingFloatChannel, keys [[8,40],[12,-40],[15,0]]}
          Scale.X {type=MovieSceneScriptingFloatChannel, keys [[8,0],[12,1],[15,1]]}
          Scale.Y {type=MovieSceneScriptingFloatChannel, keys [[8,0],[12,1.5],[15,1]]}
         ]
        }]
      }]
    }
   ]}
 }]
graphs [EventGraph {nodes [
   Event Construct()
   > PlayAnimation(InAnimation=Get Intro())
   Event OnInitialized()
   > Delay(Duration=5)
   > ClearEventState!(
    execute
    self:Gameplay Event Router Component Object Reference=GetGlobalEventRouter()
    EventType:Script Struct Object Reference=/Script/BattleRoyaleGameplayAlerts.FortEvent_CurrentGameplayAlertChanged
   ) -> (then)
   Event Tick()
  ]}]
variables []
widgets {
 info {inheritedWidgetCount=0, parentClass=&UserWidget, rootWidgetClass=&GridPanel, widgetCount=8}
 widgets [
  GridPanel_0 {widgetClassPath=&GridPanel, widget=GridPanel_0}
  Overlay_Icon {
   widgetClassPath=&Overlay
   parent=GridPanel_0
   widget=Overlay_Icon
   slot {class=GridSlot, properties {column=1, horizontalAlignment=HAlign_Center, verticalAlignment=VAlign_Top}}
  }
  Image_HomerIcon {
   widgetClassPath=&Image
   bIsVariable=true
   parent=Overlay_Icon
   widget=Image_HomerIcon
   properties {
    brush {
     bIsDynamicallyLoaded=false
     drawAs=Image
     imageType=FullColor
     mirroring=NoMirror
     resourceName=None
     resourceObject=null
     tiling=NoTile
     imageSize=<168 168>
     margin=<ltrb 0 0 0 0>
     outlineSettings {
      bUseBrushTransparency=false
      roundingType=HalfHeightRadius
      width=0
      color {colorUseRule=UseColor_Specified, specifiedColor=<rgba 0 0 0 0>}
      cornerRadii=<0 0 0 0>
     }
     tintColor {colorUseRule=UseColor_Specified, specifiedColor=<rgba 1 1 1 1>}
     uVRegion {bIsValid=false, max=<0 0>, min=<0 0>}
    }
    renderTransform {angle=0, scale=<0.2 0.2>, shear=<0 0>, translation=<0 0>}
   }
   slot {class=OverlaySlot, properties {horizontalAlignment=HAlign_Center, verticalAlignment=VAlign_Center}}
  }
  Dots {
   widgetClassPath=&Image
   bIsVariable=true
   parent=Overlay_Icon
   widget=Dots
   properties {
    brush {
     bIsDynamicallyLoaded=false
     drawAs=Image
     imageType=FullColor
     mirroring=NoMirror
     resourceName=None
     resourceObject=null
     tiling=NoTile
     imageSize=<168 168>
     margin=<ltrb 0 0 0 0>
     outlineSettings {
      bUseBrushTransparency=false
      roundingType=HalfHeightRadius
      width=0
      color {colorUseRule=UseColor_Specified, specifiedColor=<rgba 0 0 0 0>}
      cornerRadii=<0 0 0 0>
     }
     tintColor {colorUseRule=UseColor_Specified, specifiedColor=<rgba 1 1 1 1>}
     uVRegion {bIsValid=false, max=<0 0>, min=<0 0>}
    }
    renderTransform {angle=0, scale=<0.2 0.2>, shear=<0 0>, translation=<0 0>}
   }
   slot {class=OverlaySlot}
  }
  Overlay_MainText {
   widgetClassPath=&Overlay
   parent=GridPanel_0
   widget=Overlay_MainText
   slot {class=GridSlot, properties {column=1, horizontalAlignment=HAlign_Center, row=1, verticalAlignment=VAlign_Center, nudge=<0 -40.7672>}}
  }
  Text_MainText {
   widgetClassPath=&CommonTextBlock
   bIsVariable=true
   parent=Overlay_MainText
   widget=Text_MainText
   properties {
    text="[PlayerName] Excommunicated"
    colorAndOpacity {colorUseRule=UseColor_Specified, specifiedColor=<rgba 0.8632 0.0147 0 1>}
    font {
     bForceMonospaced=false
     bMaterialIsStencil=false
     fontMaterial=null
     fontObject=&Roboto
     letterSpacing=40
     monospacedWidth=1
     size=42
     skewAmount=0
     typefaceFontName=Bold
     outlineSettings {
      bApplyOutlineToDropShadows=false
      bMiteredCorners=false
      bSeparateFillAlpha=false
      outlineMaterial=null
      outlineSize=2
      outlineColor=<rgba 0 0 0 1>
     }
    }
    shadowColorAndOpacity=<rgba 0 0 0 1>
    shadowOffset=<0 2>
   }
   slot {class=OverlaySlot, properties {horizontalAlignment=HAlign_Center, verticalAlignment=VAlign_Center}}
  }
  Overlay_SecondaryText {
   widgetClassPath=&Overlay
   parent=GridPanel_0
   widget=Overlay_SecondaryText
   slot {class=GridSlot, properties {column=1, horizontalAlignment=HAlign_Center, row=2, verticalAlignment=VAlign_Center, nudge=<0 -60.1005>}}
  }
  Text_MainText_1 {
   widgetClassPath=&CommonTextBlock
   bIsVariable=true
   parent=Overlay_SecondaryText
   widget=Text_MainText_1
   properties {
    text=Track them down for a reward.
    colorAndOpacity {colorUseRule=UseColor_Specified, specifiedColor=<rgba 0.8632 0.0147 0 1>}
    font {
     bForceMonospaced=false
     bMaterialIsStencil=false
     fontMaterial=null
     fontObject=&Roboto
     letterSpacing=0
     monospacedWidth=1
     size=24
     skewAmount=0
     typefaceFontName=Bold
     outlineSettings {
      bApplyOutlineToDropShadows=false
      bMiteredCorners=false
      bSeparateFillAlpha=false
      outlineMaterial=null
      outlineSize=2
      outlineColor=<rgba 0 0 0 1>
     }
    }
    shadowColorAndOpacity=<rgba 0 0 0 1>
    shadowOffset=<0 2>
   }
   slot {class=OverlaySlot, properties {horizontalAlignment=HAlign_Center, verticalAlignment=VAlign_Center}}
  }
 ]
}
